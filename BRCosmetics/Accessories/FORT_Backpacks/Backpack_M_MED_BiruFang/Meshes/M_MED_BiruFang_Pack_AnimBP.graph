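# Backpack animation Blueprint for BiruFang that floats and spins a crystal on the pack, speeding up the spin after the wearer gets a kill.
# The AnimGraph offsets the pack, then layers random-vector drift, spring-based Anim Dynamics and additive rotation on crystal_mstr.
# The EventGraph watches kill and damage stats, with new kills driving a timed spin boost through a wrapping accumulator function.
# Damage, ring movement and speed-based values never reach the AnimGraph, and several disconnected experiments remain.
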AnimBlueprint /BRCosmetics/Accessories/FORT_Backpacks/Backpack_M_MED_BiruFang/Meshes/M_MED_BiruFang_Pack_AnimBP
target_skeleton=&/M_MED_BiruFang_Pack_Skeleton
graphs [
 AnimGraph {nodes [
   n1 = "Transform (Modify) Bone - Bone: pack"(ComponentPose=Local To Component(LocalPose=Input Pose()), Translation="0,-1,0", Rotation="0,0,5") {
    rotationMode=BMM_Additive
    rotationSpace=BCS_BoneSpace
    translationMode=BMM_Additive
    alphaBoolBlend {blendOption=Linear}
    boneToModify {boneName=pack}
    rotation {roll=5}
    translation {y=-1}
   }
   n2 = "Transform (Modify) Bone - Bone: crystal_mstr"(ComponentPose=@n1, Alpha=1, Translation="0,0.4,0", Rotation="0,0,0", Scale="1,1,1") {
    rotationSpace=BCS_WorldSpace
    translationMode=BMM_Replace
    translationSpace=BCS_ParentBoneSpace
    alphaBoolBlend {blendOption=Linear}
    alphaScaleBiasClamp {interpSpeedDecreasing=2, interpSpeedIncreasing=2}
    boneToModify {boneName=crystal_mstr}
    translation {y=0.4}
   }
   n3 = "Transform (Modify) Bone - Bone: crystal_mstr"(ComponentPose=@n2, Alpha=1, Translation=Get Globe movement() -> (Globe movement:Vector), Rotation="0,0,0", Scale="1,1,1") {
    rotationSpace=BCS_WorldSpace
    translationMode=BMM_Additive
    translationSpace=BCS_BoneSpace
    alphaBoolBlend {blendOption=Linear}
    alphaScaleBiasClamp {interpSpeedDecreasing=2, interpSpeedIncreasing=2}
    boneToModify {boneName=crystal_mstr}
   }
   n4 = Anim Dynamics - Bone: crystal_mstr(ComponentPose=@n3, Alpha=0.5, ExternalForce="0,0,0") {
    angularDampingOverride=1
    angularSpringConstant=20000
    bAngularSpring=true
    bGravityOverrideInSimSpace=true
    bLinearSpring=true
    bOverrideAngularDamping=true
    bOverrideLinearDamping=true
    bUseGravityOverride=true
    gravityScale=0.35
    linearDampingOverride=0.9999
    linearSpringConstant=50
    simulationSpace=RootRelative
    alphaBoolBlend {blendOption=Linear}
    boundBone {boneName=crystal_mstr}
    componentAppliedLinearAccClamp=<30000 30000 30000>
    componentLinearAccScale=<0.3 0.3 0.4>
    componentLinearVelScale=<0.3 0.3 0.4>
    gravityOverride {z=-980}
    physicsBodyDefinitions [{
      sphereCollisionRadius=0
      boundBone {boneName=crystal_mstr}
      boxExtents=<5 10 10>
      constraintSetup {angularLimitsMax=<0 0 0>, angularLimitsMin=<0 0 0>, angularTarget=<1 0 0>, linearAxesMax=<50 1 1>, linearAxesMin=<0 -1 -1>}
      localJointOffset=<0 0 1>
     }]
   }
   n5 = "Transform (Modify) Bone - Bone: crystal_mstr"(ComponentPose=@n4, Alpha=1, Translation="0,0,0", Rotation=Get Crystal Rotation() -> (Crystal Rotation:Rotator), Scale="1,1,1") {
    rotationMode=BMM_Additive
    rotationSpace=BCS_ParentBoneSpace
    translationMode=BMM_Additive
    translationSpace=BCS_ParentBoneSpace
    alphaBoolBlend {blendOption=Linear}
    boneToModify {boneName=crystal_mstr}
   }
   Output Pose(Result=Component To Local(ComponentPose=@n5)) {layerGroup=None}
  ]}
 EventGraph {nodes [
   n1 = Spin Function!(
    execute
    self
    Spin:double=Get Spin A() -> (Spin A:double)
    Speed:double=Get Speed A() -> (Speed A:double)
    Velocity:double=Get Velocity A() -> (Velocity A:double)
   ) -> (then, Output:double)
   > n2 = Set Spin A(Spin A:double=@n1) -> (Output_Get:double)
   > Set Spin Result A(Spin Result A:Rotator=MakeRotator(Roll=@n2)) -> (Output_Get:Rotator)
   Event OnCharacterCustomizationLoaded()
   > n3 = Sequence()
   > n4 = Cast To FortPlayerPawn(Object=TryGetFortPawn())
   > n5 = WatchCosmeticStat!(
    execute:Exec
    Pawn:Fort Player Pawn Object Reference=@n4
    StatTag:Fort Stat Manager Tag Structure=(TagName="StatManager.AthenaKills")
    DefaultValue:int=0
    bBroadcastInitialValue:bool=true
    InitialValueBroadcastDelay:float=-1
    ChangeValueBroadcastDelay:float=-1
    bConsumeWaitingNotifications:bool=true
   ) -> (then:Exec, OnStatChanged:Exec, StatName:Name, StatValue:int) {is=title K2Node_AsyncAction}
   n6 = Set Kill Count(execute=@n5.OnStatChanged, Kill Count:double="To Float (Integer)"(InInt=@n5.StatValue)) -> (Output_Get:double)
   > Branch(Condition="float > float"(A=@n6, B=Get Last Kill Count() -> (Last Kill Count:double)) {is=title K2Node_CallFunction})
   > Set Active Kill(Active Kill:bool=true) -> (Output_Get:bool)
   > Set Last Kill Count(Last Kill Count:double=@n6) -> (Output_Get:double)
   n7 = Cast To FortPlayerPawn(execute=@n3.then_1, Object=TryGetFortPawn())
   > n8 = WatchCosmeticStat(
    Pawn=@n7
    StatTag=(TagName="StatManager.DamageDealt")
    DefaultValue=0
    bBroadcastInitialValue=true
    InitialValueBroadcastDelay=-1
    ChangeValueBroadcastDelay=-1
    bConsumeWaitingNotifications=true
   )
   n9 = Set Damage Count(execute=@n8.OnStatChanged, Damage Count:double="To Float (Integer)"(InInt=@n8.StatValue)) -> (Output_Get:double)
   > Branch(Condition="float > float"(A=@n9, B=Get Last Damage Count() -> (Last Damage Count:double)))
   > Set Active Damage(Active Damage:bool=true) -> (Output_Get:bool)
   > Set Last Damage Count(Last Damage Count:double=@n9) -> (Output_Get:double)
   > Set Crank Count(Crank Count:double=float + float(A=Get Crank Count() -> (Crank Count:double), B=-180) {is=title K2Node_CommutativeAssociativeBinaryOperator}) -> (Output_Get:double)
   Event BlueprintUpdateAnimation()
   > n10 = Sequence!(execute) -> (then_0, then_1, then_2:Exec, then_3:Exec, then_4:Exec, then_5:Exec, then_6:Exec)
   > Set Delta Time Remap(Delta Time Remap:double=FInterpTo(Current=Get Delta Time Remap() -> (Delta Time Remap:double), Target=1, DeltaTime=0, InterpSpeed=1)) -> (Output_Get:double)
   n11 = OR Boolean!(A=Get bIsFalling(), B=Get bIsSkyDiving(), C:bool=Get TacSprint() -> (TacSprint:bool)) -> (ReturnValue)
   > n12 = Set Speed Alpha(Speed Alpha:double=FInterpTo(
     Current=Get Speed Alpha() -> (Speed Alpha:double)
     Target=SelectFloat(A=1, B=0, bPickA=@n11)
     DeltaTime=Reroute Node(InputPin:double) -> (OutputPin:double)
     InterpSpeed=4
    )) -> (Output_Get:double)
   n13 = Get bIsSkydiveDiveMode()
   > Set Velocity A(Velocity A:double=float * float(
     A=FInterpEaseInOut!(A:double=1.15, B:double=6.4, Alpha:double=@n12, Exponent:double=3) -> (ReturnValue:double)
     B=SelectFloat(A=1.75, B=1, bPickA=@n13)
    ) {is=title K2Node_PromotableOperator}) -> (Output_Get:double)
   n14 = Random Vector Function!(
    execute=@n10.then_1
    CharacterPartAnimInstance:Custom Character Part Anim Instance Object Reference
    SkelMeshComponent:Mesh Component Object Reference=GetOwnerSkeletalMeshComponent()
    JointName:Name=None
    GameTimeInSeconds:double=GetGameTimeInSeconds()
    TimeOffset:double=0
    PerlinScale:double=0
    Frequency:Vector="1,0,0"
    Amplitude:Vector="1,0,0"
    Offset:Vector="0,0,0"
   ) -> (then, VectorOut:Vector)
   > Set Globe movement(Globe movement:Vector=@n14) -> (Output_Get:Vector)
   n15 = Random Vector Function(
    execute=@n10.then_3
    CharacterPartAnimInstance=Self-Reference()
    JointName=None
    GameTimeInSeconds=GetGameTimeInSeconds()
    TimeOffset=0
    PerlinScale=0
    Frequency="0.7,0.8,0.5"
    Amplitude="0.2,0.5,0.3"
    Offset="0,0,0"
   )
   > Set Ring Movement(Ring Movement:Vector=@n15) -> (Output_Get:Vector)
   n16 = Get Active Kill() -> (Active Kill:bool)
   n17 = Branch(execute=@n10.then_4, Condition=@n16)
   n18 = SelectFloat(A=1.05, B=0.315, bPickA=@n16)
   n19 = Get Crystal Spin() -> (Crystal Spin:double)
   > n20 = Spin Function(Spin=@n19, Speed=@n18, Velocity=@n18)
   > n21 = Set Crystal Spin(Crystal Spin:double=@n20) -> (Output_Get:double)
   > Set Crystal Rotation(Crystal Rotation:Rotator=MakeRotator(Roll=@n21)) -> (Output_Get:Rotator)
   > Delay(Duration=5)
   > Set Active Kill(Active Kill=false)
   n22 = Spin Function(execute=@n17.else, Spin=@n19, Speed=@n18, Velocity=@n18)
   > n23 = Set Crystal Spin(Crystal Spin=@n22)
   > Set Crystal Rotation(Crystal Rotation=MakeRotator(Roll=@n23))
   n24 = Branch(execute=@n10.then_2, Condition=Get bIsSprinting())
   n25 = Vector Length(A=Get PawnVelocity())
   > Set TacSprint(TacSprint:bool="Equal (Float)"(A=MapRangeClamped(Value=@n25, InRangeA=550, InRangeB=600), B=1)) -> (Output_Get:bool)
   Set TacSprint(execute=@n24.else, TacSprint=false)
   n26 = Branch(Condition=Get Active Kill())
   Set Crystal_Movement_Alpha(execute=[@n26.then, @n39.then], Crystal_Movement_Alpha:double=1) -> (Output_Get:double)
   > n27 = Spin Function(Spin=Get Spin A(), Speed=Get Speed A(), Velocity=Get Velocity A())
   > n28 = Set Spin A(Spin A=@n27)
   > Set Spin Result A(Spin Result A=MakeRotator(Roll=@n28))
   > Delay(Duration=6)
   > Set Active Kill(Active Kill=false)
   Set Crystal_Movement_Alpha(execute=@n26.else, Crystal_Movement_Alpha=0)
   Branch(Condition=Get Crystal Rewind() -> (Crystal Rewind:bool))
   > n29 = Set Crystal_Movement_Alpha(Crystal_Movement_Alpha=float - float(A=Get Crystal_Movement_Alpha() -> (Crystal_Movement_Alpha:double), B=0) {is=title K2Node_CallFunction})
   > Branch(Condition="float <= float"!(A:double=@n29, B:double=720) -> (ReturnValue:bool) {is=title K2Node_CallFunction})
   > Set Crystal Rewind(Crystal Rewind:bool=false) -> (Output_Get:bool)
   n30 = Branch(Condition=Get Active Kill())
   > n31 = Spin Function(Spin=Get Spin A(), Speed=Get Speed A(), Velocity=Get Velocity A())
   > n32 = Set Spin A(Spin A=@n31)
   > Set Spin Result A(Spin Result A=MakeRotator(Roll=@n32))
   > Set Crystal_Timer(Crystal_Timer:double=float + float(A=Get Crystal_Timer() -> (Crystal_Timer:double), B=1)) -> (Output_Get:double)
   n33 = Get Crystal Timer Length() -> (Crystal Timer Length:double)
   n34 = Get Crystal_Timer()
   > Set Crystal_Movement_Alpha(Crystal_Movement_Alpha=MapRangeClamped(Value=float - float(A=@n33, B=@n34), InRangeA=Get Crystal Timer Length(), InRangeB=0, OutRangeA=1, OutRangeB=0))
   > Delay(Duration=4.5)
   > Branch(Condition="Equal (Float)"!(A:double=Get Crystal_Movement_Alpha(), B:double=0) -> (ReturnValue:bool) {is=title K2Node_CallFunction})
   > Set Active Kill(Active Kill=false)
   > Set Crystal_Timer(Crystal_Timer=0)
   > Delay(Duration=0.1)
   > Set Crystal_Timer(Crystal_Timer=1)
   > Set Crystal Rewind(Crystal Rewind=true)
   > Delay(Duration=0.4)
   Set Spin Result A(execute=@n30.else, Spin Result A=MakeRotator())
   n35 = Branch(Condition=Get Active Kill())
   n36 = GetGameTimeInSeconds()
   > Set Crystal Rotation(Crystal Rotation=MakeRotator(Roll=float * float(A=@n36, B=80) {is=title K2Node_CommutativeAssociativeBinaryOperator}))
   > Set Crystal_Timer(Crystal_Timer=float + float(A=Get Crystal_Timer(), B=1))
   n37 = Get Crystal Timer Length()
   n38 = Get Crystal_Timer()
   > Set Crystal_Movement_Alpha(Crystal_Movement_Alpha=MapRangeClamped(Value=float - float(A=@n37, B=@n38), InRangeA=Get Crystal Timer Length(), InRangeB=0, OutRangeA=1, OutRangeB=0))
   > Delay(Duration=4.5)
   > Set Active Kill(Active Kill=false)
   > Set Crystal_Timer(Crystal_Timer=0)
   > Delay(Duration=0.1)
   > Set Crystal_Timer(Crystal_Timer=1)
   > Set Crystal Rewind(Crystal Rewind=true)
   > Delay(Duration=0.4)
   Set Crystal Rotation(execute=@n35.else, Crystal Rotation=MakeRotator())
   Get!(execute:Exec) -> (then:Exec, else:Exec, Skel_Mesh:Skeletal Mesh Component Object Reference)
   > Sequence!(execute) -> (then_0, then_1)
   > Branch(Condition=Get Active Kill())
   > PrintString!(execute, InString=Kill, bPrintToScreen, bPrintToLog, TextColor, Duration, Key) -> (then)
   > Set Active Kill(Active Kill=true)
   n39 = Set Crystal_Timer(Crystal_Timer=float + float(A=Get Crystal_Timer(), B=1))
   Set Active Kill(Active Kill=false)
   n40 = GetGameTimeInSeconds()
   > Set Crystal Rotation(Crystal Rotation=MakeRotator(Yaw=float * float(A=@n40, B=-80)))
   RandomFloatInRange(Min=0, Max=1)
   TryGetPawnOwner()
   MapRangeClamped(
    Value=float - float(A=Get Crystal Timer Length(), B=Get Crystal_Timer())
    InRangeA=Get Crystal Timer Length()
    InRangeB=0
    OutRangeA=1
    OutRangeB=0
   )
   GetRealTimeSeconds()
   FInterpEaseInOut(A=1.5, B=0.5, Alpha=1, Exponent=1)
   FInterpEaseInOut(A=1.5, B=0.5, Alpha=1, Exponent=1)
   float * float(A="Sin (Degrees)"!(A:double=0) -> (ReturnValue:double), B=360)
   Branch(Condition="Equal (Float)"(A=Get Crystal_Movement_Alpha(), B=0))
   Do Once()
   Get bIsSprinting()
   Set Spin Result A(Spin Result A=MakeRotator())
  ]}
 Spin Function {nodes [
   n1 = Spin Function!() -> (then, Spin:double=~, Speed:double=~, Velocity:double=~) {is=title K2Node_FunctionEntry}
   > Set Velocity Mult(Velocity Mult:double=@n1.Velocity) -> (Output_Get:double)
   > n2 = Set Spin Var(Spin Var:double=@n1) -> (Output_Get:double)
   > n3 = Branch(Condition="float >= float"(A=@n2, B=720) {is=title K2Node_CallFunction})
   > n4 = Set Spin Var(Spin Var=float - float(A=Get Spin Var() -> (Spin Var:double), B=720))
   n5 = Get Velocity Mult() -> (Velocity Mult:double)
   n6 = Get Delta Time Remap()
   n7 = Set Spin Var(
    execute=@n3.else
    Spin Var=float + float!(A=Get Spin Var(), B=float * float(A=@n1.Speed, B=@n6), C:double=float * float(A=1, B=@n5)) -> (ReturnValue)
   )
   Return Node!(execute=[@n7.then, @n4.then], Output:double=@n7)
  ]}
]
variables [
 Active Damage {type=boolean}
 Active Kill {type=boolean}
 Crank Count {type=number}
 Crystal Rewind {type=boolean}
 Crystal Rotation {type=Rotator}
 Crystal Spin {type=number}
 Crystal Timer Length {type=number}
 Crystal_Movement_Alpha {type=number}
 Crystal_Timer {type=number}
 Damage Count {type=number}
 Delta Time Remap {type=number}
 Did Kill {type=boolean}
 Globe movement {type=Vector}
 Kill Count {type=number}
 Last Damage Count {type=number}
 Last Kill Count {type=number}
 Pawn {type=FortPlayerPawn}
 Ring Movement {type=Vector}
 Skel_Mesh {type=SkeletalMeshComponent}
 Speed A {type=number}
 Speed Alpha {type=number}
 Spin A {type=number}
 Spin Result A {type=Rotator}
 Spin Var {type=number}
 TacSprint {type=boolean}
 Velocity A {type=number}
 Velocity Mult {type=number}
]
